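# Physics asset for the VectorSpark backpack cosmetic, a creature with head, ears, face fur, paws, legs and tail whose dyn_ bones can move physically.
# Capsule bodies cover each bone, with two on the head.
# Limit constraints form a tree from Body: head to ears and face fur, paws to mid-paws, legs to toes, and a three-segment tail chain.
# Swing and twist limits are tight, and the paw-to-body joints lock twist.
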PhysicsAsset /BRCosmetics/Accessories/FORT_Backpacks/BackPack_M_Med_VectorSpark/Meshes/M_MED_VectorSpark_Pack_Physics
bodies {
 Body [{shapeType=Capsule, length=4.25, radius=8.25, shapeName=Body_capsule, center=<2.8938 -0.6063 0>, rotation=<rot -90 0 20.0012>}]
 dyn_Ear_L [{shapeType=Capsule, length=1.5, radius=2, shapeName=dyn_Ear_R_capsule, center=<2.9574 -0.0001 0.5609>, rotation=<rot 90 0 -179.9983>}]
 dyn_Ear_R [{shapeType=Capsule, length=1.5, radius=2, shapeName=dyn_Ear_R_capsule, center=<2.8174 0 -0.5609>, rotation=<rot -90 179.9993 179.9998>}]
 dyn_FaceFur_L [{shapeType=Capsule, length=1, radius=2, shapeName=dyn_Ear_R_capsule, center=<1.8727 -0.0001 0.0304>, rotation=<rot 90 0 -179.9984>}]
 dyn_FaceFur_R [{shapeType=Capsule, length=1, radius=2, shapeName=dyn_Ear_R_capsule, center=<1.8426 0 -0.0304>, rotation=<rot -90 179.9993 179.9998>}]
 dyn_Head [
  {shapeType=Capsule, length=0.75, radius=7, shapeName=dyn_Head_capsule, center=<4.7225 -1.6151 0>}
  {shapeType=Capsule, length=2, radius=2.75, shapeName=dyn_Head_capsule, center=<3.2628 -7.9333 0>, rotation=<rot 0 -20.0009 90>}
 ]
 dyn_Paw_L [{shapeType=Capsule, length=3.25, radius=2.5, shapeName=dyn_Paw_R_capsule, center=<3.2479 0.3107 0>, rotation=<rot -90 -90 90.0005>}]
 dyn_Paw_Mid_L [{shapeType=Capsule, length=1.25, radius=2, shapeName=dyn_Paw_Mid_R_capsule, center=<1.2658 0.4736 0>, rotation=<rot -90 180 150.001>}]
 dyn_Paw_Mid_R [{shapeType=Capsule, length=1.25, radius=2, shapeName=dyn_Paw_Mid_R_capsule, center=<1.2658 0.4736 0>, rotation=<rot -90 180 150.001>}]
 dyn_Paw_R [{shapeType=Capsule, length=3.25, radius=2.5, shapeName=dyn_Paw_R_capsule, center=<3.2479 0.3107 0>, rotation=<rot -90 -90 90.0005>}]
 dyn_Tail1 [{shapeType=Capsule, length=4.25, radius=3.5, shapeName=dyn_Tail1_capsule, center=<-0.3814 -0.59 2.208>, rotation=<rot 5.1624 1.8205 -12.4496>}]
 dyn_Tail2 [{
   shapeType=Capsule
   length=3.5
   radius=2.75
   shapeName=dyn_Tail2_capsule
   center=<0.9873 0.2468 2.5109>
   rotation=<rot -18.6349 -0.3669 5.773>
  }]
 dyn_TailStart [{
   shapeType=Capsule
   length=2.5
   radius=2.5
   shapeName=dyn_TailStart_capsule
   center=<-1.9859 -0.2708 2.0726>
   rotation=<rot 43.7538 -0.9269 -9.9086>
  }]
 dyn_Toe_L [{shapeType=Capsule, length=1, radius=2.25, shapeName=dyn_Toe_R_capsule, center=<0.7864 0 0>, rotation=<rot 90 180 89.9993>}]
 dyn_Toe_R [{shapeType=Capsule, length=1, radius=2.25, shapeName=dyn_Toe_R_capsule, center=<0.8177 0 0>, rotation=<rot 90 -31.0797 58.9196>}]
 dyn_leg_L [{shapeType=Capsule, length=4, radius=2.25, shapeName=dyn_leg_R_capsule, center=<2.5863 1.0056 0>, rotation=<rot -90 177.6141 162.3873>}]
 dyn_leg_R [{shapeType=Capsule, length=4, radius=2.25, shapeName=dyn_leg_R_capsule, center=<2.5982 1.0987 0>, rotation=<rot -90 180 160.0006>}]
}
constraints [
 {bone1Name=dyn_Head, bone2Name=Body, swing1LimitDegrees=30, swing2LimitDegrees=30, twistLimitDegrees=40}
 {bone1Name=dyn_Ear_R, bone2Name=dyn_Head, swing1LimitDegrees=25, swing2LimitDegrees=30, twistLimitDegrees=5}
 {bone1Name=dyn_FaceFur_R, bone2Name=dyn_Head, swing1LimitDegrees=10, swing2LimitDegrees=30, twistLimitDegrees=5}
 {bone1Name=dyn_leg_R, bone2Name=Body, swing1LimitDegrees=40, swing2LimitDegrees=30, twistLimitDegrees=5}
 {bone1Name=dyn_Toe_R, bone2Name=dyn_leg_R, swing1LimitDegrees=10, swing2LimitDegrees=40, twistLimitDegrees=2}
 {bone1Name=dyn_Ear_L, bone2Name=dyn_Head, swing1LimitDegrees=25, swing2LimitDegrees=30, twistLimitDegrees=5}
 {bone1Name=dyn_FaceFur_L, bone2Name=dyn_Head, swing1LimitDegrees=10, swing2LimitDegrees=30, twistLimitDegrees=5}
 {bone1Name=dyn_Paw_R, bone2Name=Body, swing1LimitDegrees=10, swing2LimitDegrees=15, twistLimitDegrees=5, twistMotion=Locked}
 {bone1Name=dyn_Paw_Mid_R, bone2Name=dyn_Paw_R, swing1LimitDegrees=25, swing2LimitDegrees=2, twistLimitDegrees=2}
 {bone1Name=dyn_Paw_L, bone2Name=Body, swing1LimitDegrees=10, swing2LimitDegrees=15, twistLimitDegrees=5, twistMotion=Locked}
 {bone1Name=dyn_Paw_Mid_L, bone2Name=dyn_Paw_L, swing1LimitDegrees=25, swing2LimitDegrees=2, twistLimitDegrees=2}
 {bone1Name=dyn_leg_L, bone2Name=Body, swing1LimitDegrees=40, swing2LimitDegrees=30, twistLimitDegrees=5}
 {bone1Name=dyn_Toe_L, bone2Name=dyn_leg_L, swing1LimitDegrees=10, swing2LimitDegrees=40, twistLimitDegrees=2}
 {bone1Name=dyn_TailStart, bone2Name=Body, swing1LimitDegrees=10, swing2LimitDegrees=10, twistLimitDegrees=10}
 {bone1Name=dyn_Tail1, bone2Name=dyn_TailStart, swing1LimitDegrees=10, swing2LimitDegrees=20, twistLimitDegrees=15}
 {bone1Name=dyn_Tail2, bone2Name=dyn_Tail1, swing1LimitDegrees=15, swing2LimitDegrees=20, twistLimitDegrees=15}
]
